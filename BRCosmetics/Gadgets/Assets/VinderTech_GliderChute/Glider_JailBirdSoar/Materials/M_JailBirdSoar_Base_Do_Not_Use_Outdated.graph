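# Base lit skeletal-mesh material for the JailBirdSoar glider, built from shared weapon shader functions and marked outdated and do-not-use by its name.
# GunMaster output runs through emissive time-of-day, then a shading path switch adds elemental effects on the default path and skips them on mobile.
# ADS scope, skin FX and creative move tool functions follow in a material-attributes chain, with World Position Offset finally overridden by a zero constant.
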Material /BRCosmetics/Gadgets/Assets/VinderTech_GliderChute/Glider_JailBirdSoar/Materials/M_JailBirdSoar_Base_Do_Not_Use_Outdated
graph {
 properties {
  bUseMaterialAttributes=true
  bUsedWithMeshDeformer=true
  bUsedWithMorphTargets=true
  bUsedWithSkeletalMesh=true
  numCustomizedUVs=5
  pixelDepthOffsetMode=PDOM_Legacy
  usedShadingModels=Default Lit
 }
 nodes [
  Constant() {r=0.5}
  n1 = MaterialFunctionCall!(UV 1 coordinates, MuzzleFalloff) {materialFunction=/Game/Packages/Fortnite_UniqueMaterials_1/Weapons/M_GunMasterShaderFunction}
  n2 = MaterialFunctionCall!(In=@n1.Material Result) {materialFunction=/Game/Weapons/WeaponSkins/MaterialFunction/WeaponEmissiveTOD}
  n3 = MaterialFunctionCall() {materialFunction=/Game/Packages/Fortress_SharedMaterials/Base_Material_Functions/BaseCustomizationFunctions/MF_Base_LocalPosition}
  n4 = MaterialFunctionCall!(Material Attributes=ShadingPathSwitch(
    Default=MaterialFunctionCall!(
     In=@n2.Result
     Pre-Skinned Local Position=@n1.Pre-Skinned Local Position Pixel Shader
     Pre-Skinned Local Normal=@n3.Pre-Skinned Local Normal Pixel Shader
    ) {materialFunction=/Game/Weapons/WeaponSkins/MaterialFunction/ElementalWeaponsFunction}.Result
    Mobile=@n2.Result
   ) {desc=Migrated from FeatureLevelSwitch}) {materialFunction=/Game/Packages/Fortnite_UniqueMaterials_1/Weapons/ADSScopeMaster}
  SetMaterialAttributes!(
   MaterialAttributes=MaterialFunctionCall(MA In=MaterialFunctionCall!(Pre-Skinned World Normal, Gmap, Customization Mask, IsCharacter, IsStatsB, Material Attributes=@n4.Result) {materialFunction=/Game/Weapons/WeaponSkins/MaterialFunction/Weapon_SkinFXFunction}.Result) {materialFunction=&MF_CreativeMoveTool}.MA Out
   World Position Offset=Constant()
  ) {attributeSetTypes [#1]}
 ]
}
